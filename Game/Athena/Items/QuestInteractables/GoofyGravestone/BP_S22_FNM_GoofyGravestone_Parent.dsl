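; Parent Blueprint for the Season 22 FNM "goofy gravestone" quest interactables, props that show interaction text to players.
; It always allows interaction and returns a zero interaction time.
; Both the normal and failed interaction prompts return the same InteractionText variable.
; BeginPlay is empty and the construction script only calls its parent; no quest logic appears in this graph.
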(event EventBeginPlay)

(fn BlueprintCanInteract
  (params (InteractingPawn :type Fort%20Pawn%20Object%20Reference) (InteractionBeingAttempted :type EInteractionBeingAttempted%20Enum :default "FirstInteraction") (InteractionType :type TInteractionType%20Enum :default "IT_NoInteraction"))
  (return true))

(fn BlueprintGetFailedInteractionString
  (params (InteractingPawn :type Fort%20Pawn%20Object%20Reference) (InteractionBeingAttempted :type EInteractionBeingAttempted%20Enum :default "FirstInteraction"))
  (return (Variables|GetInteractionText)))

(fn BlueprintGetInteractionString
  (params (InteractingPawn :type Fort%20Pawn%20Object%20Reference) (InteractionBeingAttempted :type EInteractionBeingAttempted%20Enum :default "FirstInteraction"))
  (return (Variables|GetInteractionText)))

(fn BlueprintGetInteractionTime
  (params (InteractingPawn :type Fort%20Pawn%20Object%20Reference) (InteractionBeingAttempted :type EInteractionBeingAttempted%20Enum :default "FirstInteraction"))
  (return 0.0 true))

(fn ConstructionScript
  (|CallParentFunction))
